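# Niagara system for the start of the SaltyBrim "Super" emote: seven colored Chaos Emeralds appear around a building golden charge-up.
# ChaosEmeralds pins lit sprites to seven gadget FX bones with staggered scale-in, and ChaosEmeraldsTrail reads those particles to draw color-matched ribbons.
# After short delays, BurstCharge ramps up inward streaks with meshes and flickering lights, while Rays adds camera-aligned rays and a growing glow.
# Spawn groups and partitions select per-layer and per-emerald values.
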NiagaraSystem /BRCosmetics/Animation/Game/MainPlayer/Emotes/SaltyBrim/FX/NS_Emote_SaltyBrim_SuperStart
dependencies {
 usedDataInterfaces [
  &NiagaraDataInterfaceSkeletalMesh
  &NiagaraDataInterfaceArrayInt32
  &NiagaraDataInterfaceCurve
  &NiagaraDataInterfaceParticleRead
  &NiagaraDataInterfaceColorCurve
  &NiagaraDataInterfaceArrayColor
  &NiagaraDataInterfaceCamera
 ]
 usedDynamicInputs [
  &MakeVector4_FromLinearColor
  &LerpLinearColors
  /Niagara/DynamicInputs/LinearColor/MakeLinearColorFromVector4
  &Vector4FromCurve
  &FNTimeOfDayScalar
 ]
 usedModules [
  &self:SystemSpawnScript.SystemScriptSource.SystemScriptGraph.NiagaraNodeAssignment_1.SetVariables_#1
  &SystemState
  &self:SystemSpawnScript.SystemScriptSource.SystemScriptGraph.NiagaraNodeAssignment_0.SetVariables_#2
 ]
}
emitters {
 BurstCharge {
  emitterUpdateScript [
   &EmitterState(Life Cycle Mode=::NewEnumerator1, Loop Behavior=::NewEnumerator1, Loop Delay=0.4, Loop Duration=2.9717, UseLoopDelay=true)
   &SpawnBurst_Instantaneous(Spawn Count=2, Spawn Group=1, Spawn Time=0.528, Use Spawn Probability=false, UseLoopCountLimit=false) {module=SpawnBurst_Instantaneous001}
   &SpawnBurst_Instantaneous(Spawn Count=2, Spawn Group=2, Spawn Time=0.528, Use Spawn Probability=false, UseLoopCountLimit=false)
   &SpawnRate(
    SpawnRate=&FloatFromCurve(
     FloatCurve=DataInterface() {propertyValues {Curve {
        defaultValue=1
        keys [
         {interpMode=RCIM_Cubic, tangentMode=RCTM_User, time=0.16, value=0}
         {arriveTangent=2.0152, interpMode=RCIM_Cubic, leaveTangent=2.0152, tangentMode=RCTM_User, time=0.64, value=1}
         <at 0.7845 3.0621 RCIM_Cubic>
        ]
       }}}
     CurveIndex=$Emitter.NormalizedLoopAge
     Scale Curve=101.5705
    )
    Use Spawn Probability=false
   )
  ]
  particleSpawnScript [
   /SetVariables(Particles.VisibilityTag=&SpawnGroup()) {isSetParameters=true}
   &PartitionParticles(Number Of Partitions=&MergeHelper_IntSwitch(0=0, 1=2, 2=2, 3=0, 4=0, 5=?, Index=&SpawnGroup(), Num=::NewEnumerator3))
   &InitializeParticle(
    Color=&MergeHelper_ColorSwitch(
     0=&SelectLinearColorFromArray(
      Array Sampling Mode=::NewEnumerator0
      Direct Array Index=?
      Color Selection Array=DataInterface() {propertyValues {ColorData [<rgba 5 3.216 0 1>, <rgba 1 0.3331 0 1>]}}
      Random Seed=0
     )
     1=<rgba 2.8125 1.2635 0.0704 0.4087>
     2=<rgba 2.8125 1.2635 0.0704 0.4087>
     3=<rgba 0 0 0 0>
     4=<rgba 0 0 0 0>
     Index=&SpawnGroup()
     Num=::NewEnumerator3
    )
    Lifetime=&MergeHelper_FloatSwitch(
     0=&RandomRangeFloat(Minimum=0.064, Maximum=0.15)
     1=2.7
     2=2.7
     3=0
     4=0
     Index=&SpawnGroup()
     Num=::NewEnumerator3
    )
    Mass Max=2
    Mass Mode=::NewEnumerator3
    Mesh Scale Mode=::NewEnumerator3
    Mesh Uniform Scale=0.92
    Sprite Size=&MergeHelper_Vector2Switch(
     0=&RandomRangeVector2D(Minimum=<2 6>, Maximum=<2 11>)
     1=<50 50>
     2=<0 0>
     3=<0 0>
     4=<0 0>
     Index=&SpawnGroup()
     Num=::NewEnumerator3
    )
    Sprite Size Mode=::NewEnumerator4
    UsePositionOffset=false
    Write Material Random=true
   )
   &ShapeLocation(
    MaskSpawnGroup=false
    Override Seed=false
    Sphere Radius=&MergeHelper_FloatSwitch(0=100, 1=0, 2=?, 3=?, Index=&SpawnGroup(), Num=::NewEnumerator0)
    Sphere Surface Distribution=1
    UseOwnerScale=true
   )
   &AddVelocity(
    Module.Use Falloff Distance=false
    Override Seed=false
    Velocity Mode=::NewEnumerator1
    Velocity Speed=&MergeHelper_FloatSwitch(
     0=&RandomRangeFloat(Minimum=-756.3997, Maximum=-157.489)
     1=0
     2=?
     3=?
     Index=&SpawnGroup()
     Num=::NewEnumerator0
    )
   )
   &InitialMeshOrientation(
    Mesh Orientation Mode=::NewEnumerator5
    Rotation=&MergeHelper_VectorSwitch(0=<0.25 0 0>, 1=<-0.25 0 0>, 2=?, 3=?, Index=$Particles.PartitionParticles.Partition, Num=::NewEnumerator0)
    Use Rotation Vector=true
   )
  ]
  particleUpdateScript [
   &ParticleState()
   &ScaleMeshSize(Scale Factor=&VectorFromFloat(Value=&FloatFromCurve(FloatCurve=DataInterface() {propertyValues {Curve {keys [<at 0.3131 1.0147>,<at 0.97 1.5032>,<at 1 0>]}}})))
   &ScaleColor(
    Scale Alpha=&MergeHelper_FloatSwitch(
     0=&FloatFromCurve(FloatCurve=DataInterface() {propertyValues {Curve {keys [<at -0.0046 1.0113 RCIM_Cubic>,<at 1.0011 0.002 RCIM_Cubic>]}}})
     1=&FloatFromCurve(
      FloatCurve=DataInterface() {propertyValues {Curve {keys [
          {arriveTangent=5.6153, interpMode=RCIM_Cubic, leaveTangent=5.6153, tangentMode=RCTM_User, time=-0.0046, value=0}
          {arriveTangent=0.0019, interpMode=RCIM_Cubic, leaveTangent=0.0019, time=0.2095, value=0.7174}
          {arriveTangent=-1.4247, interpMode=RCIM_Cubic, leaveTangent=-1.4247, tangentMode=RCTM_User, time=1.0011, value=0.002}
         ]}}}
      Scale Curve=1.523
     )
     2=0
     3=0
     4=0
     Index=&SpawnGroup()
     Num=::NewEnumerator3
    )
    ScaleA=true
    ScaleRGB=true
   )
   &SolveForcesAndVelocity(Clamp Velocity=false, Limit Acceleration=false)
   &DynamicMaterialParameters(
    Index 0 Param 2=&FloatFromCurve(FloatCurve=DataInterface() {propertyValues {Curve {keys [<at 0.086 1>,<at 1 0>]}}})
    Index 0 Param 3=-0.4464
    Index 0 Param 4=1.0618
    Index 1 Param 1=1
    Index 1 Param 2=1
    Index 1 Param 3=1
    Index 1 Param 4=1
    Index 2 Param 1=1
    Index 2 Param 2=1
    Index 2 Param 3=1
    Index 2 Param 4=1
    Index 3 Param 1=1
    Index 3 Param 2=1
    Index 3 Param 3=1
    Index 3 Param 4=1
    Module.Index0Float0Write=true
    Module.Index0Float1Write=true
    Module.Index0Float2Write=true
    Module.Index0Float3Write=true
    Module.Index1Float0Write=true
    Module.Index1Float1Write=true
    Module.Index1Float2Write=true
    Module.Index1Float3Write=true
    Module.Index2Float0Write=true
    Module.Index2Float1Write=true
    Module.Index2Float2Write=true
    Module.Index2Float3Write=true
    Module.Index3Float0Write=true
    Module.Index3Float1Write=true
    Module.Index3Float2Write=true
    Module.Index3Float3Write=true
    Write Parameter Index 1=::NewEnumerator1
    Write Parameter Index 2=::NewEnumerator1
    Write Parameter Index 3=::NewEnumerator1
   )
   &FN_Light_Attributes(
    Color=&Multiply_LinearColorByFloat(
     Linear Color=<rgba 5 2.7183 0 1>
     Float=&FloatFromCurve(
      FloatCurve=DataInterface() {propertyValues {Curve {
         defaultValue=1
         keys [
          {interpMode=RCIM_Cubic, tangentMode=RCTM_User, time=0, value=0}
          {arriveTangent=1.2422, interpMode=RCIM_Cubic, leaveTangent=1.2422, time=0.1473, value=0.9417}
          {interpMode=RCIM_Cubic, tangentMode=RCTM_User, time=0.7737, value=0.9612}
          {tangentMode=RCTM_User, time=1, value=0}
         ]
        }}}
      Scale Curve=&Waveform(
       "Waveform [1]"=::NewEnumerator0
       "[1] Frequency"=&FloatFromCurve(FloatCurve=DataInterface() {propertyValues {Curve {defaultValue=1, keys [{tangentMode=RCTM_User, time=0.0151, value=0.5}, {tangentMode=RCTM_User, time=1, value=3}]}}})
       "[1] Amplitude Min/Max"=<0.8 1.5>
      )
     )
    )
    Create Color Attribute=true
    Create Position Attribute=true
    Create Radius Attribute=true
    Position=&MergeHelper_VectorSwitch(0=<0 0 91.2437>, 1=<0 0 -108.2347>, 2=?, 3=?, Index=$Particles.PartitionParticles.Partition, Num=::NewEnumerator0)
    Radius=&FloatFromCurve(
     FloatCurve=DataInterface() {propertyValues {Curve {
        defaultValue=1
        keys [
         {interpMode=RCIM_Cubic, tangentMode=RCTM_User, time=0, value=0.44}
         {arriveTangent=0.1341, interpMode=RCIM_Cubic, leaveTangent=0.1341, tangentMode=RCTM_User, time=0.1546, value=1}
         {interpMode=RCIM_Cubic, tangentMode=RCTM_User, time=0.772, value=1}
         {tangentMode=RCTM_User, time=1, value=0}
        ]
       }}}
     Scale Curve=76.3986
    )
    Scale Color=1.1636
   )
  ]
  rendererClasses [
   &NiagaraSpriteRendererProperties
   &NiagaraMeshRendererProperties
   &NiagaraLightRendererProperties
  ]
  renderers [
   {rendererClass=&NiagaraSpriteRendererProperties}
   {rendererClass=&NiagaraMeshRendererProperties, rendererIndex=1}
   {rendererClass=&NiagaraLightRendererProperties, rendererIndex=2}
  ]
 }
 ChaosEmeralds {
  emitterUpdateScript [
   &self:InterpolateSpeedThresholdEmitter_01(
    Define Initial Value=false
    Rate Of Change=5
    Target Float Value:NiagaraFloat=&MakeCustomFloatFromBool(
     Bool=&SetBoolByFloatComparison(
      A=&Add_Float(
       A=&Add_Float(
        A=&Abs_Float(Float:NiagaraFloat=&MakeFloatFromVector(Vector=$Engine.Owner.Velocity, Channel=::NewEnumerator0))
        B=&Abs_Float(Float:NiagaraFloat=&MakeFloatFromVector(Vector=$Engine.Owner.Velocity, Channel=::NewEnumerator2))
       )
       B=&Abs_Float(Float:NiagaraFloat=&MakeFloatFromVector(Vector=$Engine.Owner.Velocity, Channel=::NewEnumerator1))
      )
      B=46.278
     )
     True Float=1
     False Float=0
    )
   ) {module=InterpolateSpeedThresholdEmitter_01}
   &EmitterState(Life Cycle Mode=::NewEnumerator1, Loop Behavior=::NewEnumerator1, Loop Duration=3, UseLoopDelay=false)
   &SpawnBurst_Instantaneous(Spawn Count=7, Use Spawn Probability=false, UseLoopCountLimit=false) {module=SpawnBurst_Instantaneous001}
   &SpawnBurst_Instantaneous(Spawn Count=7, Spawn Group=1, Use Spawn Probability=false, UseLoopCountLimit=false) {module=SpawnBurst_Instantaneous002}
  ]
  particleSpawnScript [
   /SetVariables(Particles.VisibilityTag=&SpawnGroup()) {isSetParameters=true}
   &PartitionParticles(Number Of Partitions=7)
   &InitializeParticle(
    Color=&Multiply_LinearColorByFloat(
     Linear Color=&MergeHelper_ColorSwitch(
      0=&MergeHelper_ColorSwitch(
       0=<rgba 0.6132 0.9782 2 1>
       1=<rgba 2 0.9913 1.9374 1>
       2=<rgba 2 0.4962 0.4513 1>
       3=<rgba 2 1.6075 0.863 1>
       4=<rgba 0.9682 2 0.6725 1>
       5=<rgba 0.8817 1 0.932 1>
       6=<rgba 1.4382 0.6852 2 1>
       Index=$Particles.PartitionParticles.Partition
       Num=::NewEnumerator5
      )
      1=&MergeHelper_ColorSwitch(
       0=<rgba 0 0.4558 2 1>
       1=<rgba 1.9193 0 2 1>
       2=<rgba 2 0 0.0802 1>
       3=<rgba 1 0.7105 0 1>
       4=<rgba 0.14 1 0 1>
       5=<rgba 0.8817 1 0.932 1>
       6=<rgba 0.4824 0 1 1>
       Index=$Particles.PartitionParticles.Partition
       Num=::NewEnumerator5
      )
      2=?
      3=?
      Index=&SpawnGroup()
      Num=::NewEnumerator0
     )
     Float=&MergeHelper_FloatSwitch(0=1.9986, 1=0.6296, 2=?, 3=?, Index=&SpawnGroup(), Num=::NewEnumerator0)
    )
    Lifetime=2.7098
    Position Mode=::NewEnumerator0
    Sprite Size=&MergeHelper_Vector2Switch(0=<73.8557 33.0651>, 1=<40 40>, 2=?, 3=?, 4=?, Index=&SpawnGroup(), Num=::NewEnumerator0)
    Sprite Size Mode=::NewEnumerator4
    Write Material Random=true
   )
  ]
  particleUpdateScript [
   &ParticleState()
   &SkeletalMeshLocation(
    Bone / Socket Index=&SelectIntFromArray(Direct Array Index=$Particles.PartitionParticles.Partition, Int Selection Array=DataInterface() {propertyValues {IntData [0,1,2,3,4,5,6]}})
    Bone Sampling Mode=::NewEnumerator4
    MaskSpawnGroup=false
    Skeletal Mesh=DataInterface() {propertyValues {
      DefaultMesh=null
      ExcludeBoneName=root
      OptionalComponentClass=null
      PreviewMesh=/BRCosmetics/Animation/Props/Gadgets_SaltyBrim/Meshes/SaltyBrim_Gadget
      SkinningMode=SkinOnTheFly
      SoftSourceActor=null
      SourceMode=Default
      UvSetIndex=0
      WholeMeshLOD=-1
      bExcludeBone=false
      bReadDeformedGeometry=true
      bRequireCurrentFrameData=true
      ComponentTags []
      FilteredBones [JNT_fx_01,JNT_fx_02,JNT_fx_03,JNT_fx_04,JNT_fx_05,JNT_fx_06,JNT_fx_07]
      FilteredSockets []
      MeshUserParameter {parameter {name=None, typeDefHandle {registeredTypeIndex=1}, varData []}}
      SamplingRegions []
     }}
   )
   &ScaleSpriteSize(
    Scale Sprite Size Mode=::NewEnumerator0
    Uniform Scale Factor=&MergeHelper_FloatSwitch(
     0=&FloatFromCurve(FloatCurve=DataInterface() {propertyValues {Curve {defaultValue=0, keys [{tangentMode=RCTM_User, time=0, value=0}, <at 0.2 1>, <at 0.913 1>, {tangentMode=RCTM_User, time=1, value=0}]}}})
     1=&FloatFromCurve(FloatCurve=DataInterface() {propertyValues {Curve {defaultValue=0, keys [{tangentMode=RCTM_User, time=0.0285, value=0}, <at 0.2 1>, <at 0.913 1>, {tangentMode=RCTM_User, time=1, value=0}]}}})
     2=&FloatFromCurve(FloatCurve=DataInterface() {propertyValues {Curve {defaultValue=0, keys [{tangentMode=RCTM_User, time=0.057, value=0}, <at 0.2 1>, <at 0.913 1>, {tangentMode=RCTM_User, time=1, value=0}]}}})
     3=&FloatFromCurve(FloatCurve=DataInterface() {propertyValues {Curve {defaultValue=0, keys [{tangentMode=RCTM_User, time=0.0855, value=0}, <at 0.2 1>, <at 0.913 1>, {tangentMode=RCTM_User, time=1, value=0}]}}})
     4=&FloatFromCurve(FloatCurve=DataInterface() {propertyValues {Curve {defaultValue=0, keys [{tangentMode=RCTM_User, time=0.114, value=0}, <at 0.2 1>, <at 0.913 1>, {tangentMode=RCTM_User, time=1, value=0}]}}})
     5=&FloatFromCurve(FloatCurve=DataInterface() {propertyValues {Curve {defaultValue=0, keys [{tangentMode=RCTM_User, time=0.1425, value=0}, <at 0.2 1>, <at 0.913 1>, {tangentMode=RCTM_User, time=1, value=0}]}}})
     6=&FloatFromCurve(FloatCurve=DataInterface() {propertyValues {Curve {defaultValue=0, keys [{tangentMode=RCTM_User, time=0.171, value=0}, <at 0.2 1>, <at 0.913 1>, {tangentMode=RCTM_User, time=1, value=0}]}}})
     Index=$Particles.PartitionParticles.Partition
     Num=::NewEnumerator5
    )
   )
   &Drag(Drag=&RandomRangeFloat(Minimum=1, Maximum=2), Use Linear Drag=true, Use Rotational Drag=false)
   &SolveForcesAndVelocity(Clamp Velocity=false, Limit Acceleration=false)
   &ScaleColor(ScaleA=true, ScaleRGB=true)
   &DynamicMaterialParameters(
    Index 0 Param 1=1.1657
    Index 0 Param 2=0.8651
    Module.Index0Float0Write=true
    Module.Index0Float1Write=true
    Module.Index0Float2Write=true
    Module.Index0Float3Write=true
   )
   &FN_Light_Attributes(
    Color=$Particles.Color
    Create Color Attribute=true
    Create Radius Attribute=true
    Radius=&FloatFromCurve(
     FloatCurve=DataInterface() {propertyValues {Curve {
        defaultValue=1
        keys [
         <at -0.0062 0.5146 RCIM_Cubic>
         {arriveTangent=0.4596, interpMode=RCIM_Cubic, leaveTangent=0.4596, time=0.1306, value=0.8058}
         {interpMode=RCIM_Cubic, tangentMode=RCTM_User, time=0.7542, value=0.8641}
         {tangentMode=RCTM_User, time=1, value=0}
        ]
       }}}
     Scale Curve=50
    )
    Scale Color=3
   )
   /SetVariables(Particles.FlareSize:Vector2f=&Multiply_Vector2D_ByFloat(
     Float=&Waveform("Waveform [1]"=::NewEnumerator0, "[1] Frequency"=10.2066, "[1] Amplitude Min/Max"=<0.8051 1.7286>)
     Vector2D=$Particles.SpriteSize
    )) {isSetParameters=true}
  ]
  rendererClasses [&NiagaraSpriteRendererProperties, &NiagaraLightRendererProperties]
  renderers [
   {rendererClass=&NiagaraSpriteRendererProperties}
   {rendererClass=&NiagaraSpriteRendererProperties, rendererIndex=1}
   {rendererClass=&NiagaraLightRendererProperties, rendererIndex=2}
   {rendererClass=&NiagaraSpriteRendererProperties, rendererIndex=3}
  ]
 }
 ChaosEmeraldsTrail {
  emitterUpdateScript [
   /SetVariables(Emitter.NewSpawnRate=16) {isSetParameters=true}
   &EmitterState(Life Cycle Mode=::NewEnumerator1, Loop Behavior=::NewEnumerator1, Loop Duration=2.3403, UseLoopDelay=false)
   &SpawnParticlesFromOtherEmitter(
    Attribute Reader=DataInterface() {propertyValues {EmitterBinding {bindingMode=Other, emitterName=ChaosEmeralds}}}
    Maximum Age Spawn Threshold=1.4
    Spawn Rate=$Emitter.NewSpawnRate
    Spawn Rate Per Particle Cap=350
    Use Maximum Age=false
    Use Minimum Age=false
   )
  ]
  particleSpawnScript [
   &PartitionParticles(Number Of Partitions=7, Write Partition to Ribbon ID=::NewEnumerator0)
   &InitializeParticle(
    Color=&MergeHelper_ColorSwitch(
     0=<rgba 0 0.4558 2 1>
     1=<rgba 1.9193 0 2 1>
     2=<rgba 2 0 0.0802 1>
     3=<rgba 1 0.7105 0 1>
     4=<rgba 0.14 1 0 1>
     5=<rgba 0.8817 1 0.932 1>
     6=<rgba 0.4824 0 1 1>
     Index=$Particles.PartitionParticles.Partition
     Num=::NewEnumerator5
    )
    Lifetime=0.394
    Ribbon Width=1
    Ribbon Width Mode=::NewEnumerator1
    UsePositionOffset=false
    Write Material Random=true
   )
   &SampleParticlesFromOtherEmitter(
    Particle ID Sampling=::NewEnumerator1
    Reject Particle If Sampled Index Is Invalid=false
    Sprite Size Sampling:ENiagara_DirectReadSamplingMode=::NewEnumerator1
   )
   &AddVelocity(Velocity=<0 0 547.6467>) {enabled=false}
  ]
  particleUpdateScript [
   &ParticleState()
   /Niagara/Modules/AttributeReader/UpdateParticlesFromOtherEmitter(
    Attribute Reader=DataInterface() {propertyValues {EmitterBinding {bindingMode=Other, emitterName=ChaosEmeralds}}}
    Position Sampling:ENiagara_DirectReadSamplingMode=::NewEnumerator1
    Sprite Size Sampling:ENiagara_DirectReadSamplingMode=::NewEnumerator1
   )
   &self:ScratchModule(Position:NiagaraPosition=$Output.UpdateParticlesFromOtherEmitter.SampledPosition) {module=ScratchModule}
   &ScaleRibbonWidth(Ribbon Width Scale=&Multiply_Float(
     A=&FloatFromCurve(FloatCurve=DataInterface() {propertyValues {Curve {keys [<at 0 1>,<at 1 0.46>]}}}, Scale Curve=0.4)
     B=&MakeFloatFromVector2d(Vector 2D=$Output.UpdateParticlesFromOtherEmitter.SampledSpriteSize)
    ))
   &InheritSourceMovement(Applied Position Scale=<0 0 0>, ApplyForce=false, ApplyPosition=true, ApplyPositionFalloff=false)
   &SolveForcesAndVelocity(Clamp Velocity=false, Limit Acceleration=false)
   &ScaleColor(
    Scale Alpha=&FloatFromCurve(
     FloatCurve=DataInterface() {propertyValues {Curve {keys [
         {arriveTangent=-0.8321, interpMode=RCIM_Cubic, leaveTangent=35.6275, tangentMode=RCTM_Break, time=0, value=0}
         {arriveTangent=0.0819, interpMode=RCIM_Cubic, leaveTangent=0.6334, tangentMode=RCTM_Break, time=0.0597, value=1}
         {arriveTangent=-1.7699, tangentMode=RCTM_User, time=1, value=0}
        ]}}}
     Scale Curve=0.128
    )
    ScaleA=true
    ScaleRGB=true
   )
   /SetVariables(
    Particles.HardColor:LinearColor=&Multiply_LinearColorByFloat(
     Linear Color=&Multiply_LinearColorByFloat(
      Linear Color=&MergeHelper_ColorSwitch(
       0=<rgba 0 0.4558 2 1>
       1=<rgba 1.9193 0 2 1>
       2=<rgba 2 0 0.0802 1>
       3=<rgba 1 0.7105 0 1>
       4=<rgba 0.14 1 0 1>
       5=<rgba 0.8817 1 0.932 1>
       6=<rgba 0.4824 0 1 1>
       Index=$Particles.PartitionParticles.Partition
       Num=::NewEnumerator5
      )
      Float=19.2127
     )
     Float=&FloatFromCurve(
      FloatCurve=DataInterface() {propertyValues {Curve {keys [
          {arriveTangent=-0.8321, interpMode=RCIM_Cubic, leaveTangent=35.6275, tangentMode=RCTM_Break, time=0, value=0}
          {arriveTangent=0.0819, interpMode=RCIM_Cubic, leaveTangent=0.6334, tangentMode=RCTM_Break, time=0.0597, value=1}
          {arriveTangent=-1.7699, tangentMode=RCTM_User, time=1, value=0}
         ]}}}
      Scale Curve=0.128
     )
    )
    Particles.HardRibbonWiudth:NiagaraFloat=&Multiply_Float(
     A=&FloatFromCurve(FloatCurve=DataInterface() {propertyValues {Curve {keys [<at 0 1>,<at 0.44 0>]}}}, Scale Curve=0.0672)
     B=&MakeFloatFromVector2d(Vector 2D=$Output.UpdateParticlesFromOtherEmitter.SampledSpriteSize)
    )
   ) {isSetParameters=true}
  ]
  rendererClasses [&NiagaraRibbonRendererProperties]
  renderers [
   {rendererClass=&NiagaraRibbonRendererProperties}
   {rendererClass=&NiagaraRibbonRendererProperties, rendererIndex=1}
  ]
 }
 FlashCover {rendererClasses [], renderers []}
 Rays {
  emitterUpdateScript [
   &EmitterState(
    Inactive Response=::NewEnumerator1
    Life Cycle Mode=::NewEnumerator1
    Loop Behavior=::NewEnumerator1
    Loop Delay=0.48
    Loop Duration=3.0784
    UseLoopDelay=true
   )
   &SpawnBurst_Instantaneous(Spawn Group=1, Use Spawn Probability=false, UseLoopCountLimit=false)
   &SpawnRate(
    SpawnRate=&FloatFromCurve(
     FloatCurve=DataInterface() {propertyValues {Curve {
        defaultValue=1
        keys [
         {interpMode=RCIM_Cubic, tangentMode=RCTM_User, time=0, value=0}
         {arriveTangent=2.0465, interpMode=RCIM_Cubic, leaveTangent=2.0465, time=0.6244, value=0.6382}
         {arriveTangent=2.0152, interpMode=RCIM_Cubic, leaveTangent=2.0152, tangentMode=RCTM_User, time=0.78, value=1.5963}
        ]
       }}}
     CurveIndex=$Emitter.NormalizedLoopAge
     Scale Curve=130
    )
    Use Spawn Probability=false
   )
  ]
  particleSpawnScript [
   /SetVariables(Particles.VisibilityTag=&SpawnGroup()) {isSetParameters=true}
   &PartitionParticles(Number Of Partitions=&MergeHelper_IntSwitch(0=2, 1=1, 2=?, 3=?, 4=?, 5=?, Index=&SpawnGroup(), Num=::NewEnumerator0))
   &InitializeParticle(
    Color=&MergeHelper_ColorSwitch(
     0=&SelectLinearColorFromArray(
      Array Sampling Mode=::NewEnumerator1
      Direct Array Index=$Particles.PartitionParticles.Partition
      Color Selection Array=DataInterface() {propertyValues {ColorData [<rgba 0.599 0.2832 0.0076 1>, <rgba 2.5 1.2273 0 1>]}}
      Random Seed=?
     )
     1=<rgba 3.2174 1.4866 0 0.0609>
     2=<rgba 0 0 0 0>
     3=<rgba 0 0 0 0>
     4=<rgba 0 0 0 0>
     Index=&SpawnGroup()
     Num=::NewEnumerator3
    )
    Lifetime=&MergeHelper_FloatSwitch(
     0=&RandomRangeFloat(Minimum=0.15, Maximum=0.35)
     1=3
     2=?
     3=?
     Index=&SpawnGroup()
     Num=::NewEnumerator0
    )
    Mesh Scale Mode=::NewEnumerator3
    Mesh Uniform Scale=0.432
    Sprite Size=&MergeHelper_Vector2Switch(
     0=&RandomRangeVector2D(Minimum=<11.3207 82.6927>, Maximum=<31.2623 36.0607>)
     1=<351.1279 351.1279>
     2=?
     3=?
     4=?
     Index=&SpawnGroup()
     Num=::NewEnumerator0
    )
    Sprite Size Mode=::NewEnumerator4
    UsePositionOffset=false
    Write Material Random=true
   )
   &SubUVAnimation(End Frame=1, SubUV Animation Mode:ENiagara_SubUVLookupMode=::NewEnumerator2)
   &ShapeLocation(
    Hemisphere Distribution=<0.5 1>
    MaskSpawnGroup=false
    Offset Mode=::NewEnumerator0
    Override Seed=false
    Rotation Mode=::NewEnumerator2
    Sphere Radius=&MergeHelper_FloatSwitch(0=11, 1=0, 2=?, 3=?, Index=&SpawnGroup(), Num=::NewEnumerator0)
    UseOwnerScale=true
    Yaw / Pitch / Roll=<0.25 0 0>
   ) {module=ShapeLocation001}
   &AddVelocity(
    Module.Use Falloff Distance=false
    Override Seed=false
    Velocity Mode=::NewEnumerator1
    Velocity Speed=&MergeHelper_FloatSwitch(0=10, 1=0, 2=0, 3=0, 4=0, Index=&SpawnGroup(), Num=::NewEnumerator3)
   )
  ]
  particleUpdateScript [
   &ParticleState()
   &ScaleMeshSize(Scale Factor=&VectorFromFloat(Value=&FloatFromCurve(FloatCurve=DataInterface() {propertyValues {Curve {defaultValue=1, keys [{tangentMode=RCTM_User, time=0, value=0}, <at 0.9579 1>, {tangentMode=RCTM_User, time=1.02, value=0}]}}})))
   &ScaleSpriteSize(
    Scale Factor=&MergeHelper_Vector2Switch(
     0=&MakeVector2D(
      X=&FloatFromCurve(FloatCurve=DataInterface() {propertyValues {Curve {keys [<at 0 0.5>]}}})
      Y=&Multiply_Float(
       A=&FloatFromCurve(
        FloatCurve=DataInterface() {propertyValues {Curve {keys [<at 0 0.5>,<at 1 1>]}}}
        Scale Curve=&FloatFromCurve(FloatCurve=DataInterface() {propertyValues {Curve {keys [<at 0.8022 1>,<at 0.9137 0>]}}}, CurveIndex=$Emitter.NormalizedLoopAge)
       )
       B=1
      )
     )
     1=&Multiply_Vector2D_ByFloat(
      Float=&FloatFromCurve(FloatCurve=DataInterface() {propertyValues {Curve {defaultValue=1, keys [{tangentMode=RCTM_User, time=0, value=0}, <at 0.9617 1>, {tangentMode=RCTM_User, time=1, value=0}]}}})
      Vector2D=<1 1>
     )
     2=<0 0>
     3=<0 0>
     4=<0 0>
     Index=&SpawnGroup()
     Num=::NewEnumerator3
    )
    Scale Sprite Size Mode=::NewEnumerator2
   )
   /SetVariables(Particles.Ray2Size:Vector2f=&MakeVector2D(
     X=&RandomRangeFloat(Minimum=50, Maximum=80)
     Y=&Multiply_Float(
      A=&RandomRangeFloat(Minimum=100, Maximum=120)
      B=&FloatFromCurve(FloatCurve=DataInterface() {propertyValues {Curve {keys [<at 0.9072 0.9882>,<at 0.93 0>]}}}, CurveIndex=$Emitter.NormalizedLoopAge)
     )
    )) {isSetParameters=true}
   &ScaleColor(
    Scale Alpha=&MergeHelper_FloatSwitch(
     0=&Multiply_Float(
      A=&FloatFromCurve(FloatCurve=DataInterface() {propertyValues {Curve {keys [{interpMode=RCIM_Cubic, tangentMode=RCTM_User, time=0, value=0}, {interpMode=RCIM_Cubic, tangentMode=RCTM_User, time=0.4, value=1}, <at 1 0>]}}})
      B=1
     )
     1=&Multiply_Float(
      A=&FloatFromCurve(FloatCurve=DataInterface() {propertyValues {Curve {keys [
           {interpMode=RCIM_Cubic, tangentMode=RCTM_User, time=0, value=0}
           {interpMode=RCIM_Cubic, tangentMode=RCTM_User, time=0.1689, value=0.9882}
           {arriveTangent=-1.1891, interpMode=RCIM_Cubic, leaveTangent=-1.1891, time=0.9414, value=0.9912}
           <at 1 0>
          ]}}})
      B=1
     )
     2=0
     3=0
     4=0
     Index=&SpawnGroup()
     Num=::NewEnumerator3
    )
    ScaleA=true
    ScaleRGB=true
   )
   &SolveForcesAndVelocity(Clamp Velocity=false, Limit Acceleration=false)
   &self:GimbalLock(Gimbal Lock Direction:NiagaraPosition=&ReturnCameraVector(Camera Vectors=::NewEnumerator4, Coordinate Space=::Local, Origin Position=$Engine.Emitter.SimulationPosition)) {module=GimbalLock}
   &SpriteFacingAndAlignment(Sprite Alignment:NiagaraPosition=$Particles.GimbalPos, WriteAlignment=true, WriteFacing=false) {module=SpriteFacingAndAlignment001}
   /SetVariables(Particles.MeshColor:LinearColor=<rgba 1 0.6811 0 1>) {isSetParameters=true}
  ]
  rendererClasses [&NiagaraSpriteRendererProperties, &NiagaraMeshRendererProperties]
  renderers [
   {rendererClass=&NiagaraSpriteRendererProperties}
   {rendererClass=&NiagaraSpriteRendererProperties, rendererIndex=1}
   {rendererClass=&NiagaraSpriteRendererProperties, rendererIndex=2}
   {rendererClass=&NiagaraMeshRendererProperties, rendererIndex=3}
  ]
 }
}
summary {
 systemName=NS_Emote_SaltyBrim_SuperStart
 emitters [
  {emitterName=ChaosEmeralds, rendererClasses [&NiagaraSpriteRendererProperties, &NiagaraLightRendererProperties]}
  {emitterName=ChaosEmeraldsTrail, rendererClasses [&NiagaraRibbonRendererProperties]}
  {
   emitterName=BurstCharge
   rendererClasses [
    &NiagaraSpriteRendererProperties
    &NiagaraMeshRendererProperties
    &NiagaraLightRendererProperties
   ]
  }
  {emitterName=Rays, rendererClasses [&NiagaraSpriteRendererProperties, &NiagaraMeshRendererProperties]}
  {emitterName=FlashCover, rendererClasses []}
 ]
 userVariables [
  User.bIsFrontEnd {type=NiagaraBool}
  User.bIsFrontEndPreview {type=NiagaraBool}
 ]
}
