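# Niagara burst effect for the SaltyBrim emote, producing an orange-gold flash with glow, streaks and sparks.
# Dark_Glow pairs sprites with fading lights, while Flare002 drives layered flash sprites and a mesh with animated wave parameters.
# Spikes uses spawn groups to split camera-aligned streaks from fast, drag-slowed, curl-noise sparks.
# System time-of-day scales brightness or alpha, front-end modifiers resize sprites, and three LWE emitters list no renderers.
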NiagaraSystem /BRCosmetics/Animation/Game/MainPlayer/Emotes/SaltyBrim/FX/NS_SaltyBrim_Burst
dependencies {
 usedDataInterfaces [
  &NiagaraDataInterfaceCurve
  &NiagaraDataInterfaceColorCurve
  &NiagaraDataInterfaceArrayColor
  &NiagaraDataInterfaceVector2DCurve
  &NiagaraDataInterfaceCamera
 ]
 usedDynamicInputs [&RandomRangeVector, &FNTimeOfDayScalar]
 usedModules [
  &self:SystemSpawnScript.SystemScriptSource.SystemScriptGraph.NiagaraNodeAssignment_0.SetVariables_#1
  &self:SystemSpawnScript.SystemScriptSource.SystemScriptGraph.NiagaraNodeAssignment_2.SetVariables_#2
  &SystemState
  &self:SystemSpawnScript.SystemScriptSource.SystemScriptGraph.NiagaraNodeAssignment_1.SetVariables_#3
 ]
}
emitters {
 BurstMeshLWE {rendererClasses [], renderers []}
 Dark_Glow {
  emitterUpdateScript [
   &EmitterState(Life Cycle Mode=::NewEnumerator1, Loop Behavior=::NewEnumerator1, Loop Duration=1, UseLoopDelay=true)
   &SpawnBurst_Instantaneous(Spawn Count=1, Use Spawn Probability=false)
   &SpawnBurst_Instantaneous(Spawn Count=2, Spawn Group=1, Use Spawn Probability=false) {module=SpawnBurst_Instantaneous001}
  ]
  particleSpawnScript [
   &PartitionParticles(Number Of Partitions=&MergeHelper_IntSwitch(0=0, 1=2, 2=?, 3=?, 4=?, 5=?, Index=&SpawnGroup(), Num=::NewEnumerator0))
   /SetVariables(Particles.VisibilityTag=&SpawnGroup()) {isSetParameters=true}
   &InitializeParticle(
    Color=<rgba 0.1094 0.0385 0.0041 1>
    Lifetime=0.708
    Position Offset:NiagaraPosition=$System.PosOffset
    Sprite Rotation Mode=::NewEnumerator2
    Sprite Size Mode=::NewEnumerator3
    Uniform Sprite Size=743.9638
    UsePositionOffset=true
    Write Material Random=true
   )
   &DynamicMaterialParameters(
    Index 0 Param 1=&RandomRangeFloat(Minimum=0.4, Maximum=0.6)
    Index 0 Param 2=1
    Index 0 Param 3=1
    Index 0 Param 4=1
    Module.Param0WriteEnabled=true
    Module.Param1WriteEnabled=true
    Module.Param2WriteEnabled=true
    Module.Param3WriteEnabled=true
   )
  ]
  particleUpdateScript [
   &ParticleState()
   &SolveForcesAndVelocity(Clamp Velocity=false, Limit Acceleration=false)
   &ScaleColor(
    Scale Alpha=&Multiply_Float(
     A=1
     B=&FloatFromCurve(FloatCurve=DataInterface() {propertyValues {Curve {keys [<at 0 0>,<at 0.2 1>,<at 0.8 1>,<at 1 0>]}}})
    )
    Scale RGB=&Multiply_VectorByFloat(Vector=<1 1 1>, Float=$System.TOD)
    ScaleA=true
    ScaleRGB=true
   )
   &CameraOffset()
   &FN_Light_Attributes(
    Color=&Multiply_LinearColorByFloat(
     Linear Color=<rgba 5 2.7183 0 1>
     Float=&FloatFromCurve(FloatCurve=DataInterface() {propertyValues {Curve {keys [<at 0 1>,<at 1 0>]}}})
    )
    Create Color Attribute=true
    Create Position Attribute=true
    Create Radius Attribute=true
    Position=&MergeHelper_VectorSwitch(0=<0 0 75.9136>, 1=<0 0 -94.023>, 2=?, 3=?, Index=$Particles.PartitionParticles.Partition, Num=::NewEnumerator0)
    Radius=66.3268
    Scale Color=1.1636
   )
  ]
  rendererClasses [&NiagaraSpriteRendererProperties, &NiagaraLightRendererProperties]
  renderers [
   {rendererClass=&NiagaraSpriteRendererProperties}
   {rendererClass=&NiagaraLightRendererProperties, rendererIndex=1}
  ]
 }
 Flare002 {
  emitterUpdateScript [
   &EmitterState(Life Cycle Mode=::NewEnumerator1, Loop Behavior=::NewEnumerator1, Loop Duration=1, UseLoopDelay=true)
   &SpawnBurst_Instantaneous(Spawn Count=1, Use Spawn Probability=false)
   &SpawnBurst_Instantaneous(Spawn Count=1, Spawn Group=1, Use Spawn Probability=false) {module=SpawnBurst_Instantaneous001}
  ]
  particleSpawnScript [
   /SetVariables(Particles.VisibilityTag=&SpawnGroup()) {isSetParameters=true}
   &InitializeParticle(
    Color=<rgba 5 2.7183 0 1>
    Lifetime=&MergeHelper_FloatSwitch(0=0.3, 1=0.748, 2=?, 3=?, Index=&SpawnGroup(), Num=::NewEnumerator0)
    Mesh Scale Mode=::NewEnumerator3
    Mesh Uniform Scale=2.0439
    Position Offset:NiagaraPosition=$System.PosOffset
    Sprite Rotation Mode=::NewEnumerator2
    Sprite Size Mode=::NewEnumerator3
    Uniform Sprite Size=498.0552
    UsePositionOffset=true
    Write Material Random=true
   )
   &DynamicMaterialParameters(
    Index 0 Param 1=&RandomRangeFloat(Minimum=0.4, Maximum=0.6)
    Index 0 Param 2=1
    Index 0 Param 3=1
    Index 0 Param 4=1
    Module.Param0WriteEnabled=true
    Module.Param1WriteEnabled=true
    Module.Param2WriteEnabled=true
    Module.Param3WriteEnabled=true
   )
  ]
  particleUpdateScript [
   &ParticleState()
   &ScaleMeshSize(Scale Factor=&VectorFromFloat(Value=&FloatFromCurve(FloatCurve=DataInterface() {propertyValues {Curve {
         defaultValue=1
         keys [
          {arriveTangent=1.6237, interpMode=RCIM_Cubic, leaveTangent=1.6237, tangentMode=RCTM_User, time=0.2042, value=1}
          {arriveTangent=0.5742, interpMode=RCIM_Cubic, leaveTangent=0.5742, time=0.3314, value=1.3207}
          {arriveTangent=0.0224, interpMode=RCIM_Cubic, leaveTangent=0.0224, tangentMode=RCTM_User, time=0.9794, value=1.4451}
         ]
        }}})))
   &SolveForcesAndVelocity(Clamp Velocity=false, Limit Acceleration=false)
   &ScaleColor(
    Scale Alpha=&Multiply_Float(
     A=$System.TOD
     B=&FloatFromCurve(FloatCurve=DataInterface() {propertyValues {Curve {keys [<at 0 0>,<at 0.2 1>,<at 0.8 1>,<at 1 0>]}}})
    )
    ScaleA=true
    ScaleRGB=true
   )
   &ScaleSpriteSize(Scale Factor=&Vector2DFromFloat(Value=&FloatFromCurve(
      FloatCurve=DataInterface() {propertyValues {Curve {keys [<at 0 1.8233 RCIM_Cubic>,<at 1 0>]}}}
      Scale Curve=&FN_FrontEndFloatModifier(Gameplay=1, Front End=0.7)
     )))
   &CameraOffset(Camera Offset Amount=36.256)
   /SetVariables(Particles.Flash size:Vector2f=&Multiply_Vector2D_ByFloat(
     Float=&Multiply_Float(
      A=685.3351
      B=&FloatFromCurve(
       FloatCurve=DataInterface() {propertyValues {Curve {keys [{arriveTangent=-0.0405, interpMode=RCIM_Cubic, leaveTangent=-0.0405, tangentMode=RCTM_User, time=0.62, value=1.7699}]}}}
       Scale Curve=&FN_FrontEndFloatModifier(Gameplay=1, Front End=0.7)
      )
     )
     Vector2D=<1 1>
    )) {isSetParameters=true}
   /SetVariables(Particles.Flash color:LinearColor=&MakeLinearColorFromVectorAndFloat(
     "Vector (RGB)"=&MakeVectorFromLinearColorRGB(Color=<rgba 1 0.5585 0 1>)
     "Float (Alpha)"=&Multiply_Float(
      A=&Multiply_Float(
       A=$System.TOD
       B=&FloatFromCurve(FloatCurve=DataInterface() {propertyValues {Curve {keys [<at 0 0>,<at 0.124 1>,<at 0.496 1>,<at 0.62 0>]}}})
      )
      B=1
     )
    )) {isSetParameters=true}
   /SetVariables(Particles.flash2size:Vector2f=&Multiply_Vector2D_ByFloat(
     Float=&Multiply_Float(
      A=292.5097
      B=&FloatFromCurve(
       FloatCurve=DataInterface() {propertyValues {Curve {keys [{arriveTangent=-0.0405, interpMode=RCIM_Cubic, leaveTangent=-0.0405, tangentMode=RCTM_User, time=0.62, value=1.7699}]}}}
       Scale Curve=&FN_FrontEndFloatModifier(Gameplay=1, Front End=0.7)
      )
     )
     Vector2D=<1 1>
    )) {isSetParameters=true}
   /SetVariables(Particles.flash2color:LinearColor=&MakeLinearColorFromVectorAndFloat(
     "Vector (RGB)"=&MakeVectorFromLinearColorRGB(Color=<rgba 1 1 1 1>)
     "Float (Alpha)"=&Multiply_Float(
      A=&Multiply_Float(
       A=$System.TOD
       B=&FloatFromCurve(FloatCurve=DataInterface() {propertyValues {Curve {keys [<at 0 0>,<at 0.124 1>,<at 0.496 1>,<at 0.62 0>]}}})
      )
      B=1
     )
    )) {isSetParameters=true}
   /SetVariables(
    Particles.DyunmParamWave:Vector4f=&MakeVector4(
     X=&FloatFromCurve(FloatCurve=DataInterface() {propertyValues {Curve {keys [
          {interpMode=RCIM_Cubic, leaveTangent=-1.5024, tangentMode=RCTM_Break, time=0.18, value=0.5}
          {arriveTangent=-0.6098, interpMode=RCIM_Cubic, leaveTangent=-0.6098, time=0.3433, value=0.1559}
          <at 1 0>
         ]}}})
     Y=&FloatFromCurve(FloatCurve=DataInterface() {propertyValues {Curve {keys [{arriveTangent=-0.3464, interpMode=RCIM_Cubic, leaveTangent=-0.3464, tangentMode=RCTM_User, time=-0.0031, value=0.9929}, <at 0.87 0.9>]}}})
     Z=&FloatFromCurve(FloatCurve=DataInterface() {propertyValues {Curve {keys [<at 0 0>,<at 0.15 1.0141>,<at 1 1.2>]}}})
     W=1
    )
    Particles.WaveColor:LinearColor=&ColorFromCurve(ColorCurve=DataInterface() {propertyValues {
       AlphaCurve {keys [<at 0 1 RCIM_Cubic>,<at 0.1959 0.1721>,<at 0.9877 0 RCIM_Cubic>]}
       BlueCurve {keys [<at 0 0.0501 RCIM_Cubic>,<at 0.1799 0.0185>]}
       GreenCurve {keys [<at 0 0.8985 RCIM_Cubic>,<at 0.1799 0.3322>]}
       RedCurve {keys [<at 0 2 RCIM_Cubic>,<at 0.1799 0.7396>]}
      }})
   ) {isSetParameters=true}
  ]
  rendererClasses [&NiagaraSpriteRendererProperties, &NiagaraMeshRendererProperties]
  renderers [
   {rendererClass=&NiagaraSpriteRendererProperties}
   {rendererClass=&NiagaraSpriteRendererProperties, rendererIndex=1}
   {rendererClass=&NiagaraSpriteRendererProperties, rendererIndex=2}
   {rendererClass=&NiagaraMeshRendererProperties, rendererIndex=3}
  ]
 }
 Spikes {
  emitterUpdateScript [
   &EmitterState(Life Cycle Mode=::NewEnumerator1, Loop Behavior=::NewEnumerator1, Loop Duration=1, UseLoopDelay=true)
   &SpawnBurst_Instantaneous(Spawn Count=23, Use Spawn Probability=false, UseLoopCountLimit=false)
   &SpawnBurst_Instantaneous(Spawn Count=41, Spawn Group=1, Use Spawn Probability=false, UseLoopCountLimit=false) {module=SpawnBurst_Instantaneous001}
  ]
  particleSpawnScript [
   /SetVariables(Particles.VisibilityTag=&SpawnGroup()) {isSetParameters=true}
   &InitializeParticle(
    Color=&SelectLinearColorFromArray(
     Array Sampling Mode=::NewEnumerator0
     Direct Array Index=?
     Color Selection Array=DataInterface() {propertyValues {ColorData [<rgba 5 3.0422 0 0.5>, <rgba 1.667 0.409 0 0.5>]}}
     Random Seed=0
    )
    Lifetime=&MergeHelper_FloatSwitch(
     0=&RandomRangeFloat(Minimum=0.464, Maximum=0.6968)
     1=&RandomRangeFloat(Minimum=0.364, Maximum=0.736)
     2=?
     3=?
     Index=&SpawnGroup()
     Num=::NewEnumerator0
    )
    Sprite Size=&MergeHelper_Vector2Switch(
     0=&RandomRangeVector2D(Minimum=<305.6159 175.1617>, Maximum=<150.0497 181.2326>)
     1=&Multiply_Vector2D_ByFloat(Float=&RandomRangeFloat(Minimum=4.961, Maximum=6.1472), Vector2D=<1 1>)
     2=<0 0>
     3=<0 0>
     4=<0 0>
     Index=&SpawnGroup()
     Num=::NewEnumerator3
    )
    Sprite Size Mode=::NewEnumerator4
    UsePositionOffset=true
    Write Material Random=true
   )
   &SubUVAnimation(End Frame=0, SubUV Animation Mode:ENiagara_SubUVLookupMode=::NewEnumerator2)
   &ShapeLocation(
    Hemisphere Distribution=<0.5 1>
    MaskSpawnGroup=true
    Offset Coordinate Space=::Local
    Offset Mode=::NewEnumerator0
    Override Seed=false
    Rotation Mode=::NewEnumerator2
    Sphere Radius=4.7439
    UseOwnerScale=true
    Yaw / Pitch / Roll=<0.25 0 0>
   )
   &AddVelocity(
    Module.Use Falloff Distance=false
    Origin Offset Coordinate Space=::Simulation
    Override Seed=false
    Velocity Mode=::NewEnumerator1
    Velocity Speed=&MergeHelper_FloatSwitch(
     0=0
     1=&RandomRangeFloat(Minimum=221.0147, Maximum=717.2778)
     2=?
     3=?
     Index=&SpawnGroup()
     Num=::NewEnumerator0
    )
   )
  ]
  particleUpdateScript [
   &ParticleState()
   &Drag(
    Drag=&MergeHelper_FloatSwitch(
     0=0
     1=&RandomRangeFloat(Minimum=2.5633, Maximum=7.3245)
     2=?
     3=?
     Index=&SpawnGroup()
     Num=::NewEnumerator0
    )
    Use Linear Drag=true
    Use Rotational Drag=false
   )
   &CurlNoiseForce(
    Bias Noise=false
    Noise Strength=&MergeHelper_FloatSwitch(0=0, 1=264.451, 2=0, 3=0, 4=0, Index=&SpawnGroup(), Num=::NewEnumerator3)
    Pan Noise=false
   )
   &SolveForcesAndVelocity(Clamp Velocity=false, Limit Acceleration=false)
   &ScaleSpriteSizeBySpeed(
    Max Scale Factor=&MergeHelper_Vector2Switch(0=<1 1>, 1=<1 1.5842>, 2=?, 3=?, 4=?, Index=&SpawnGroup(), Num=::NewEnumerator0)
    Min Scale Factor=&MergeHelper_Vector2Switch(0=<1 1>, 1=<1 2.3119>, 2=?, 3=?, 4=?, Index=&SpawnGroup(), Num=::NewEnumerator0)
   )
   /SetVariables(Particles.DotScale:Vector2f=&Vector2DFromCurve(
     Vector2Curve=DataInterface() {propertyValues {
       XCurve {keys [{tangentMode=RCTM_User, time=0, value=1}, <at 0.7653 0.9082>, {tangentMode=RCTM_User, time=1, value=0}]}
       YCurve {keys [{tangentMode=RCTM_User, time=0, value=0.5}, <at 0.5931 1.7242>, {tangentMode=RCTM_User, time=0.9916, value=0}]}
      }}
     ScaleCurve=&Multiply_Vector2D_ByFloat(
      Float=&FN_FrontEndFloatModifier(Gameplay=0.6, Front End=0.5)
      Vector2D=&Multiply_Vector2D(A=<0.5 0.5>, B=$Particles.SpriteSize)
     )
    )) {isSetParameters=true}
   &ScaleSpriteSize(
    Scale Factor=&MergeHelper_Vector2Switch(
     0=&Vector2DFromCurve(
      Vector2Curve=DataInterface() {propertyValues {
        XCurve {keys [{tangentMode=RCTM_User, time=0, value=1}, {tangentMode=RCTM_User, time=1, value=0.8}]}
        YCurve {keys [<at -0.024 0>, {tangentMode=RCTM_User, time=1.0065, value=1.1816}]}
       }}
      ScaleCurve=<1 1>
     )
     1=&Vector2DFromCurve(
      Vector2Curve=DataInterface() {propertyValues {
        XCurve {keys [{tangentMode=RCTM_User, time=0, value=1}, {tangentMode=RCTM_User, time=1, value=0.8}]}
        YCurve {keys [{tangentMode=RCTM_User, time=0, value=0.5}, {tangentMode=RCTM_User, time=1, value=1.7}]}
       }}
      ScaleCurve=<1 1>
     )
     2=<0 0>
     3=<0 0>
     4=<0 0>
     Index=0
     Num=::NewEnumerator3
    )
    Scale Sprite Size Mode=::NewEnumerator2
   )
   &ScaleColor(
    Scale Alpha=&MergeHelper_FloatSwitch(
     0=&FloatFromCurve(FloatCurve=DataInterface() {propertyValues {Curve {keys [<at 0 0>,<at 0.1 1>,<at 0.2 0.8>,<at 1 0>]}}}, Scale Curve=$System.TOD)
     1=1
     2=?
     3=?
     Index=&SpawnGroup()
     Num=::NewEnumerator0
    )
    ScaleA=true
    ScaleRGB=true
   )
   &self:GimbalLock(Gimbal Lock Direction:NiagaraPosition=&ReturnCameraVector(Camera Vectors=::NewEnumerator4, Coordinate Space=::Local, Origin Position=$Particles.Position)) {module=GimbalLock}
   &SpriteFacingAndAlignment(
    Sprite Alignment:NiagaraPosition=$Particles.GimbalPos
    Sprite Facing:NiagaraPosition=$Particles.GimbalPos
    WriteAlignment=true
    WriteFacing=false
   )
  ]
  rendererClasses [&NiagaraSpriteRendererProperties]
  renderers [
   {rendererClass=&NiagaraSpriteRendererProperties}
   {rendererClass=&NiagaraSpriteRendererProperties, rendererIndex=1}
  ]
 }
 WindBurstLWE {rendererClasses [], renderers []}
 WindCircularLWE {rendererClasses [], renderers []}
}
summary {
 systemName=NS_SaltyBrim_Burst
 emitters [
  {emitterName=Flare002, rendererClasses [&NiagaraSpriteRendererProperties, &NiagaraMeshRendererProperties]}
  {emitterName=Dark_Glow, rendererClasses [&NiagaraSpriteRendererProperties, &NiagaraLightRendererProperties]}
  {emitterName=Spikes, rendererClasses [&NiagaraSpriteRendererProperties]}
  {emitterName=WindCircularLWE, rendererClasses []}
  {emitterName=WindBurstLWE, rendererClasses []}
  {emitterName=BurstMeshLWE, rendererClasses []}
 ]
 userVariables [
  User.bIsFrontEnd {type=NiagaraBool}
  User.bIsFrontEndPreview {type=NiagaraBool}
 ]
}
